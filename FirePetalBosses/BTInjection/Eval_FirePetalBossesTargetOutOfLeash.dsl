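; Behavior tree injection evaluator for Fire Petal bosses, gating a target-out-of-leash behavior, apparently the return to the spawn cloud.
; On enter it caches the bot controller and boss pawn.
; It requires the parent's check plus a challenged boss that is alive, not DBNO, outside the spawn cloud and not blocked from returning.
; A DBNO blackboard target bypasses the parent's result, keeping the other conditions.
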(event EventOnEnter
  (params (OwnerComp :type Behavior%20Tree%20Component%20Object%20Reference))
  (|CallParentFunction OwnerComp)
  (bind _ascontroller (Utilities|Casting|CastToController (Components|GetOwner OwnerComp)))
  (bind _asfort_athena_aibot_controller (Utilities|Casting|CastToFortAthenaAIBotController _ascontroller))
  (bind _output_get (Variables|Default|SetBossController _asfort_athena_aibot_controller))
  (bind _asbp_player_pawn_fire_petal_boss (Utilities|Casting|CastToBP_PlayerPawn_FirePetal_Boss (Pawn|GetControlledPawn _output_get)))
  (Variables|Default|SetBossPawn _asbp_player_pawn_fire_petal_boss))

(fn EvaluateStartingConditions
  (params (OwnerComp :type Behavior%20Tree%20Component%20Object%20Reference))
  (bind _bosspawn (Variables|Default|GetBossPawn))
  (bind _returnvalue (Health|IsDBNO _bosspawn))
  (bind _returnvalue_2 (not _returnvalue))
  (bind _returnvalue_3 (FortPawn|IsDead _bosspawn))
  (bind _returnvalue_4 (not _returnvalue_3))
  (bind _returnvalue_5 (GameplayTags|HasMatchingGameplayTag _bosspawn "(TagName=\"FirePetalBosses.Demons.IsinSpawnCloud\")"))
  (bind _returnvalue_6 (GameplayTags|HasMatchingGameplayTag _bosspawn "(TagName=\"FirePetalBosses.Demons.BlockReturnToSpawnCloud\")"))
  (bind _returnvalue_7 (not _returnvalue_5))
  (bind _returnvalue_8 (not _returnvalue_6))
  (bind _returnvalue_9 (GameplayTags|HasMatchingGameplayTag _bosspawn "(TagName=\"FirePetalBosses.Demons.IsChallenged\")"))
  (bind _returnvalue_10 (|CallParentFunction OwnerComp))
  (bind _targetactor (Class|BPFLAICommonFunctions|GetTargetActorfromBlackboard (Variables|Default|GetBossPawn)))
  (Utilities|IsValid _targetactor
    (:Is%20Valid
      (if (or (not (Utilities|IsValid _targetactor :!class /Script/Engine.KismetSystemLibrary)) (Health|IsDBNO _targetactor))
        (return (Math|Boolean|ANDBoolean _returnvalue_9 _returnvalue_7 _returnvalue_2 _returnvalue_8 _returnvalue_4))
        (else
          _)))
    (:Is%20Not%20Valid
      _))
  (return (Math|Boolean|ANDBoolean _returnvalue_10 _returnvalue_9 _returnvalue_7 _returnvalue_2 _returnvalue_8 _returnvalue_4)))
