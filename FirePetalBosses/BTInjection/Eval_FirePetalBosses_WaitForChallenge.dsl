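; Behavior tree evaluator for Fire Petal bosses that gates a wait-for-challenge state while the boss sits in its spawn cloud.
; On enter, it casts the owner to a Fort Athena bot controller and caches it along with the controlled boss pawn.
; Starting and ongoing conditions share CanWait, which requires the pawn's IsinSpawnCloud gameplay tag without IsChallenged.
; CanWait also returns true when the cached pawn is invalid.
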(event EventOnEnter
  (params (OwnerComp :type Behavior%20Tree%20Component%20Object%20Reference))
  (bind _ascontroller (Utilities|Casting|CastToController (Components|GetOwner OwnerComp)))
  (bind _asfort_athena_aibot_controller (Utilities|Casting|CastToFortAthenaAIBotController _ascontroller))
  (bind _output_get (Variables|Default|SetBossController _asfort_athena_aibot_controller))
  (bind _asbp_player_pawn_fire_petal_boss (Utilities|Casting|CastToBP_PlayerPawn_FirePetal_Boss (Pawn|GetControlledPawn _output_get)))
  (Variables|Default|SetBossPawn _asbp_player_pawn_fire_petal_boss))

(fn CanWait
  (bind _bosspawn (Variables|Default|GetBossPawn))
  (Utilities|IsValid _bosspawn
    (:Is%20Valid
      (return (and (GameplayTags|HasMatchingGameplayTag _bosspawn "(TagName=\"FirePetalBosses.Demons.IsinSpawnCloud\")") (not (GameplayTags|HasMatchingGameplayTag _bosspawn "(TagName=\"FirePetalBosses.Demons.IsChallenged\")")))))
    (:Is%20Not%20Valid
      (return true))))

(fn EvaluateOngoingConditions
  (params (OwnerComp :type Behavior%20Tree%20Component%20Object%20Reference))
  (bind _bool (CallFunction|CanWait))
  (return _bool))

(fn EvaluateStartingConditions
  (params (OwnerComp :type Behavior%20Tree%20Component%20Object%20Reference))
  (bind _bool (CallFunction|CanWait))
  (return _bool))
